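; Blueprint for the Stormking blunt hammer melee weapon, judging by its name and folder under FORT_Melee.
; Its equipped-and-attached event only calls the parent implementation.
; The construction script also only calls the parent, so neither graph adds logic of its own.
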(event FortWeapon|EventOnEquippedAndAttached
  (|CallParentFunction))

(fn ConstructionScript
  (|CallParentFunction))
